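; Athena event actor for version 5.0 that carries crack, rainbow and river-tint components, with its crack test and setup events marked disabled pending a 5.1 rework.
; OnRep handlers reveal the central crack, start its looping audio, and feed crack progression into material opacity and an audio parameter.
; Test events derive progression from days remaining or a timeline.
; The construction script places the rainbow and writes river tint to the landscape material parameter collection from in-editor testing flags.
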(event Custom|TestCracks
  (params (DaysRemaining :type Float%20%28double-precision%29))
  ;[1 0 0] Disabled for 5.0, will rework for 5.1
  ;;[1 1 1] Test crack progression by days remaining
  (Variables|Default|SetCrackProgression (Math|Float|MapRangeClamped DaysRemaining (Variables|Default|GetCrackEventDuration) 0.0 1.0))
  (Variables|Default|SetAreCracksVisible 1))

(event Custom|TestCracksAnimated
  ;[1 0 0] Disabled for 5.0, will rework for 5.1
  ;;[1 1 1] Test crack spread animation
  (Variables|Default|SetAreCracksVisible 1)
  (bind (_direction _newtrack_0) (Timeline|CrackSpreadTest PlayFromStart)
    (:Update
      (Variables|Default|SetCrackProgression _newtrack_0))))

(event Custom|ContinueSetup
  ;[1 0 0] Disabled for 5.0, will rework for 5.1
  (Utilities|FlowControl|Delay 1.0))

(event EventBeginPlay)

(fn OnRep_AreCracksVisible
  (switch int (Variables|Default|GetAreCracksVisible)
    (:1
      (Rendering|SetVisibility (Variables|Default|GetCentralCrack) true true)
      (if (not (Audio|Components|Audio|IsPlaying (Variables|Default|GetCrackAudioLoop)))
        (Audio|Components|Audio|Play (Variables|Default|GetCrackAudioLoop) 0.0)))))

(fn OnRep_CrackProgression
  (bind _crackprogression (Variables|Default|GetCrackProgression))
  (Rendering|Material|SetScalarParameterValueonMaterials (Variables|Default|GetCentralCrack) "Opacity" _crackprogression)
  (Audio|Parameter|SetFloatParameter (Variables|Default|GetCrackAudioLoop) "CrackProgression" (Math|Float|Clamp(Float) _crackprogression) :!class /Script/Engine.AudioComponent))

(fn OnRep_CracksVisible
  (if (Variables|Default|GetCracksVisible)
    (Rendering|SetVisibility (Variables|Default|GetCentralCrack) true true)
    (if (not (Audio|Components|Audio|IsPlaying (Variables|Default|GetCrackAudioLoop)))
      (Audio|Components|Audio|Play (Variables|Default|GetCrackAudioLoop) 0.0))))

(fn SetupDynamicMaterials
  (params (Mesh :type Static%20Mesh%20Component%20Object%20Reference) (MatArray :type Array%20of%20Material%20Instance%20Dynamic%20Object%20References))
  (for _index (range (Utilities|Array|Length (Rendering|Material|GetMaterials Mesh)))
    (bind _returnvalue (Rendering|Material|CreateDynamicMaterialInstance Mesh _index 0 "None" :!class /Script/Engine.PrimitiveComponent))
    (Utilities|Array|Add MatArray _returnvalue)))

(fn SetupRiverTint
  (params (EnableTint :type Boolean) (TintColor :type Linear%20Color%20Structure))
  (if EnableTint
    (Rendering|Material|SetVectorParameterValue "/Game/Athena/Environments/Landscape/MPC/MPC_Landscape.MPC_Landscape" "RiverTintAdd" TintColor :!class /Script/Engine.KismetMaterialLibrary)
    (else
      (Rendering|Material|SetVectorParameterValue "/Game/Athena/Environments/Landscape/MPC/MPC_Landscape.MPC_Landscape" "RiverTintAdd" (Utilities|Struct|MakeLinearColor 0.0) :!class /Script/Engine.KismetMaterialLibrary))))

(fn UpdateCrackProgression
  (params (Mesh :type Mesh%20Component%20Object%20Reference) (Progression :type Float%20%28double-precision%29) (Threshold :type Float%20%28double-precision%29))
  (Rendering|SetVisibility Mesh (> Progression Threshold) true)
  (Rendering|Material|SetScalarParameterValueonMaterials Mesh "CrackSize" (Math|Curves|GetFloatValue (Variables|Default|GetCrackGrowthCurve) (Math|Float|Clamp(Float) (- Progression Threshold)))))

(fn ConstructionScript
  (bind _rainbow (Variables|Default|GetRainbow))
  (Utilities|FlowControl|Sequence
    (:then_2
      (Rendering|SetVisibility _rainbow (Variables|InEditorTesting|GetShowRainbow) false)
      (Transformation|SetWorldTransform _rainbow (Variables|Default|GetRainbowWorldTransform) false false))
    (:then_3
      (CallFunction|SetupRiverTint (Variables|InEditorTesting|GetEnableRiverTint) (Variables|Default|GetRiverTintStPatrick's)))))
